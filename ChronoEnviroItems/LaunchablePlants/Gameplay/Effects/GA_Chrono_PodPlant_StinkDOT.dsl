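; Gameplay ability for the launchable pod plant's stink damage-over-time effect, activated from an event sent by the stink cloud trigger.
; On activation it reads a delay-before-damage-tick value from a scalable float row.
; A positive delay starts a timer that calls StartDamageTick; otherwise StartDamageTick runs immediately.
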(event Ability|EventActivateAbilityFromEvent
  (params (EventData :type Gameplay%20Event%20Data%20Structure%20%28by%20ref%29))
  (bind _row_delaybeforedamagetick (Variables|Default|GetRow_DelayBeforeDamageTick))
  (bind _returnvalue (ScalableFloat|GetValueatLevel _row_delaybeforedamagetick))
  ;[0.4845 1 0.2088] Activate ability from stink cloud trigger
  (if (> _returnvalue 0)
    (Utilities|Time|SetTimerbyEvent (EventDispatchers|CreateEvent :!event StartDamageTick) _returnvalue)
    (else
      (CallFunction|StartDamageTick))))
